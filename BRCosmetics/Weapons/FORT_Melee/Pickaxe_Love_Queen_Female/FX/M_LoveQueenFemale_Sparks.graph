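# Additive, unlit, two-sided material for Niagara mesh particles that draws sparkles for the Love Queen Female pickaxe effect.
# A Sprite function builds billboard sparks, sized by a vertex-color red remap between minimum and maximum parameters.
# An anime flare tinted by particle color feeds HDR derivation and depth fade.
# A panning second-UV spark texture, raised to the NumberOfSparks power, masks that result before emissive adjustments, controlling how many sparks show.
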Material /BRCosmetics/Weapons/FORT_Melee/Pickaxe_Love_Queen_Female/FX/M_LoveQueenFemale_Sparks
graph {
 properties {
  bUsedWithNiagaraMeshParticles=true
  blendMode=BLEND_Additive
  pixelDepthOffsetMode=PDOM_Legacy
  shadingModel=MSM_Unlit
  twoSided=true
  usedShadingModels=Unlit
 }
 nodes [
  n1 = MaterialFunctionCall(
   Input=VertexColor().R
   Input Low=Constant()
   Input High=Constant() {r=1}
   Target Low=ScalarParameter() {defaultValue=0.8, parameterName=size_min}
   Target High=ScalarParameter() {defaultValue=1.1, parameterName=size_max}
  ) {materialFunction=&RemapValueRange}
  n2 = ScalarParameter() {defaultValue=1, parameterName=RandomSize}
  MaterialFunctionCall!(
   Center Location=WorldPosition().XYZ
   0-1 UVs
   XY scale=Multiply(A=ScalarParameter() {defaultValue=10, parameterName=SparksSize}, B=Multiply(A=@n1.Result, B=@n2) {constB=3})
   OverridePivot
   "Normals (Optional)"
  ) {materialFunction=&Sprite}
  n3 = TextureSample() {texture=/Game/Effects/Fort_Effects/Textures/Flares/T_Anime_Flare}
  n4 = Multiply(A=ParticleColor().RGB, B=@n3.G)
  n5 = MaterialFunctionCall(LDR Input=Clamp(None=@n4), HDR Tint=@n4) {materialFunction=&DeriveHDRfromLDR}
  n6 = ScalarParameter() {defaultValue=100, parameterName=FadeDistance}
  n7 = VectorParameter() {parameterName=PannerSpeeed, defaultValue=<rgba -0.045 0 0 1>}
  n8 = Add(A=TextureCoordinate() {coordinateIndex=1}, B=Multiply(A=Time(), B=AppendVector(A=@n7.R, B=@n7.R)))
  n9 = TextureSample(UVs=@n8) {texture=&/t_lovequeen_sparkPlayer}
  n10 = ScalarParameter() {defaultValue=1.7, parameterName=NumberOfSparks}
  MaterialFunctionCall(Emissive=Multiply(A=DepthFade(Opacity=@n5.RemappedRange, FadeDistance=@n6), B=Power(Base=@n9.R, Exp=@n10) {constExponent=4})) {materialFunction=&MF_FX_EmissiveAdjustments}
  Multiply(A=Clamp(None=VertexColor())) {constB=0.1}
  n11 = MaterialFunctionCall(Distance=ScalarParameter() {defaultValue=200, parameterName=CameraDistanceFade}) {materialFunction=&Near_Camera_Fade}
  Multiply(A=@n3.G, B=QualitySwitch(Default=@n11.Result, Low=Constant() {r=1}, High=@n11.Result, Medium=@n11.Result))
  n12 = VertexColor()
  Multiply(A=@n12.R, B=DynamicParameter() {paramNames [Scale,Expand,RandomMult,Param4]}.RandomMult)
  Multiply(A=@n12.R, B=ScalarParameter() {defaultValue=1, parameterName=Param_2})
  TextureSample(UVs=@n8) {texture=/Game/Accessories/FORT_Backpacks/Backpack_F_MED_Bistro_Monster/Textures/T_F_MED_Bistro_Monster_Gradient}
 ]
}
