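# Physics asset for the Love Queen Female two-handed melee pickaxe skeletal mesh.
# One capsule body covers the weapon handle and a smaller offset capsule covers the heart joint.
# A single constraint links the heart to the handle, with a wide first swing limit and near-zero second swing and twist limits, suggesting a hinge-like dangle on one axis.
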PhysicsAsset /BRCosmetics/Weapons/FORT_Melee/Pickaxe_Love_Queen_Female/Meshes/SK_Pickaxe_Love_Queen_Female_Physics
bodies {
 Melee_TwoHanded_Handle [{shapeType=Capsule, length=86.2526, radius=25.8987, center=<0.1819 5.841 37.0507>, rotation=<rot -0.0136 -90 -0.122>}]
 heartJnt [{shapeType=Capsule, length=25.25, radius=10, shapeName=heartJnt_capsule, center=<0 -20.0269 0>, rotation=<rot 0 1 0>}]
}
constraints [{bone1Name=heartJnt, bone2Name=Melee_TwoHanded_Handle, swing1LimitDegrees=180, swing2LimitDegrees=0.001, twistLimitDegrees=0.001}]
