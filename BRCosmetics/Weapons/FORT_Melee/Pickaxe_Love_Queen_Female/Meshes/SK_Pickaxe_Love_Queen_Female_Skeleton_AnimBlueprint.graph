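# Animation Blueprint for the Love Queen Female pickaxe that switches between a spinning heart bone and rigid-body physics.
# When bIsAccelerating2D is true it blends to an additive heartJnt rotation, otherwise to RigidBody simulation, weighted lower during jumps or falls.
# Each update, SpinFunction advances crankSpin, wrapping at 720, to drive CrankRotator's yaw.
# Disconnected heartRotate, spinHeartFunction and deltaTimeRemap nodes remain in the EventGraph.
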AnimBlueprint /BRCosmetics/Weapons/FORT_Melee/Pickaxe_Love_Queen_Female/Meshes/SK_Pickaxe_Love_Queen_Female_Skeleton_AnimBlueprint
target_skeleton=&/SK_Pickaxe_Love_Queen_Female_Skeleton
graphs [
 AnimGraph {nodes [
   n1 = Get bIsJumping()
   n2 = Get bIsFalling()
   n3 = RigidBody(ComponentPose=Local To Component(LocalPose=Input Pose()), Alpha=SelectFloat(A=0.25, B=1, bPickA=OR Boolean(A=@n1, B=@n2))) {
    alpha=0
    bOverrideWorldGravity=true
    overridePhysicsAsset=&/SK_Pickaxe_Love_Queen_Female_Physics
    simulationSpace=BaseBoneSpace
    alphaBoolBlend {blendOption=Linear}
    alphaScaleBiasClamp {bInterpResult=true}
    baseBoneRef {boneName=b_AR_Root}
    componentLinearAccScale=<2 2 2>
    componentLinearVelScale=<2 2 2>
    overrideWorldGravity {z=-980}
   }
   n4 = Get bIsAccelerating2D()
   n5 = "Transform (Modify) Bone - Bone: heartJnt"!(ComponentPose, bAlphaBoolEnabled:bool=@n4, Translation="0,0,0", Rotation=Get CrankRotator() -> (CrankRotator:Rotator), Scale="1,1,1") -> (Pose) {
    alphaInputType=Bool
    rotationMode=BMM_Additive
    rotationSpace=BCS_BoneSpace
    alphaBoolBlend {blendOption=Linear}
    boneToModify {boneName=heartJnt}
   }
   Output Pose(Result=Blend Poses by bool(
     BlendPose_0=Component To Local(ComponentPose=@n5)
     BlendPose_1=Component To Local(ComponentPose=@n3)
     BlendTime_0=0.5
     BlendTime_1=0.5
     bActiveValue=@n4
    ) {blendType=Linear}) {layerGroup=None}
  ]}
 EventGraph {nodes [
   Delay(Duration=3)
   n1 = GetRealTimeSeconds()
   > Set heartRotate(heartRotate:Rotator=MakeRotator(Yaw=float * float(A=@n1, B=0))) -> (Output_Get:Rotator)
   n2 = spinHeartFunction!(execute, self, spin:double=Get crankSpin() -> (crankSpin:double), speed:double=1.25) -> (then, output:double)
   > n3 = Set crankSpin(crankSpin:double=@n2) -> (Output_Get:double)
   n4 = GetRealTimeSeconds()
   > Set heartRotate(heartRotate=MakeRotator(Yaw=float * float(A=@n4, B=500)))
   Event BlueprintUpdateAnimation()
   > n5 = SpinFunction!(execute, self, Spin:double=Get crankSpin(), Speed:double=1.25) -> (then, Output:double)
   > n6 = Set crankSpin(crankSpin=@n5)
   > Set CrankRotator(CrankRotator:Rotator=MakeRotator(Yaw=@n6)) -> (Output_Get:Rotator)
   TryGetPawnOwner()
   MakeRotator(Yaw=@n3)
   Branch(Condition=Get bIsPlayingMeleeAnim())
   Set heartRotate(heartRotate="0, 0, 0")
   Set deltaTimeRemap(deltaTimeRemap:double=FInterpTo(Current=Get deltaTimeRemap() -> (deltaTimeRemap:double), Target=1, DeltaTime=0, InterpSpeed=1)) -> (Output_Get:double)
  ]}
 SpinFunction {nodes [
   n1 = SpinFunction!() -> (then, Spin:double=~, Speed:double=~) {is=title K2Node_FunctionEntry}
   > n2 = Set spinVar(spinVar:double=@n1) -> (Output_Get:double)
   > n3 = Branch(Condition="float >= float"(A=@n2, B=720) {is=title K2Node_CallFunction})
   > n4 = Set spinVar(spinVar=float - float(A=Get spinVar() -> (spinVar:double), B=720) {is=title K2Node_CallFunction})
   n5 = Get velocity() -> (velocity:double)
   n6 = Get velocityMult() -> (velocityMult:double)
   n7 = Get deltaTimeRemap()
   n8 = Set spinVar(
    execute=@n3.else
    spinVar=float + float!(A=Get spinVar(), B=float * float(A=@n1.Speed, B=@n7), C:double=float * float(A=@n5, B=@n6)) -> (ReturnValue) {is=title K2Node_CommutativeAssociativeBinaryOperator}
   )
   Return Node!(execute=[@n4.then, @n8.then], Output:double=@n8)
  ]}
 spinHeartFunction {nodes [
   n1 = spinHeartFunction!() -> (then, spin:double=~, speed:double=~) {is=title K2Node_FunctionEntry}
   > n2 = Set spinVar(spinVar=@n1)
   > n3 = Branch(Condition="float >= float"(A=@n2, B=720))
   > n4 = Set spinVar(spinVar=float - float(A=Get spinVar(), B=720))
   n5 = Get velocity()
   n6 = Get velocityMult()
   n7 = Get deltaTimeRemap()
   n8 = Set spinVar(execute=@n3.else, spinVar=float + float(A=Get spinVar(), B=float * float(A=@n1.speed, B=@n7), C=float * float(A=@n5, B=@n6)))
   Return Node!(execute=[@n4.then, @n8.then], output:double=@n8)
  ]}
]
variables [
 CrankRotator {type=Rotator, default=<rot 0 20.3125 0>}
 crankSpin {type=number, default=20.3125}
 deltaTimeRemap {type=number, default=7}
 heartRotate {type=Rotator}
 isRotating {type=boolean}
 spinVar {type=number, default=20.3125}
 velocity {type=number}
 velocityMult {type=number}
]
